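# Character master material for the blue Dark Pit backpack, rendering dark animated wisps over a glowing center.
# A panned, sine/cosine-warped swirl texture feeds a flow map that distorts noise and normals, with noise-offset time varying the motion.
# Wisp and falloff masks blend glow with wisp color into HDR emissive, wrapped in shared character FX functions.
# The inverted two-sided-sign opacity mask clips front faces, showing only back faces.
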Material /BRCosmetics/Accessories/FORT_Backpacks/Backpack_M_MED_Dark_Pit/Skins/Blue/Materials/M_MED_Dark_Pit_Blue_Backpack_Master
graph {
 properties {
  bAutomaticallySetUsageInEditor=false
  bUseMaterialAttributes=true
  bUsedWithMeshDeformer=true
  bUsedWithMorphTargets=true
  bUsedWithSkeletalMesh=true
  blendMode=BLEND_Masked
  pixelDepthOffsetMode=PDOM_Legacy
  refractionMethod=RM_IndexOfRefraction
  shadingModel=MSM_FromMaterialExpression
  subsurfaceProfile=&SS_HeroSkin_02
  twoSided=true
  usedShadingModels=Default Lit | Subsurface | Clear Coat | Subsurface Profile | Cloth | Eye
 }
 nodes [
  n1 = ComponentMask(None=TextureCoordinate()) {g=true}
  Add(A=@n1) {constB=0.9}
  Add(A=Multiply(A=NamedRerouteUsage()) {constB=0.04}, B=@n1)
  n2 = MaterialFunctionCall() {materialFunction=&BaseFN_CharacterFunction_PreFX}
  n3 = NamedRerouteUsage()
  n4 = Panner(Coordinate=Multiply(A=TextureCoordinate()) {constB=0.75}) {speedY=0.075}
  n5 = Multiply(A=Time()) {constB=0.3}
  n6 = MaterialFunctionCall(Float2=Frac(None=Add(A=@n4, B=@n5))) {materialFunction=&BreakOutFloat2Components}
  n7 = MaterialFunctionCall(X=Sine(None=@n6.R), Y=Cosine(None=@n6.G)) {materialFunction=&MakeFloat2}
  n8 = TextureSample(UVs=Add(A=@n4, B=Multiply(A=@n7.Result) {constB=0.0075})) {
   samplerSource=SSM_Wrap_WorldGroupSettings
   samplerType=SAMPLERTYPE_LinearColor
   texture=/BRCosmetics/Accessories/FORT_Backpacks/Backpack_M_MED_Dark_Pit/Skins/Blue/Textures/T_DarkPit_SwirlFlow
  }
  n9 = ComponentMask(None=@n8.RGB) {g=true, r=true}
  n10 = Time()
  n11 = TextureSample() {samplerSource=SSM_Wrap_WorldGroupSettings, samplerType=SAMPLERTYPE_Grayscale, texture=&LowResBlurredNoise}
  n12 = MaterialFunctionCall!(
   Diffuse=TextureObject() {texture=/Game/Global/Textures/Default/Noise/TankMoteNoise}
   Normal Map=TextureObject() {samplerType=SAMPLERTYPE_Normal, texture=/Game/Global/Textures/Default/Noise/T_VIperGlitchNoise_N}
   "Flow Vector Map (see tooltip)"=Multiply(A=ConstantBiasScale(None=@n9) {bias=-0.5, scale=2}, B=ScalarParameter() {defaultValue=0.1, group=DarkPit, parameterName=DP_FlowmapStrength}) {constB=0.3}
   UV texture 2 offset=Constant3Vector()
   "Time (see tooltip)"=Multiply(A=Add(A=@n10, B=@n11.G)) {constB=0.3}
   UVs=Multiply(A=TextureCoordinate(), B=ScalarParameter() {defaultValue=0.5, group=DarkPit, parameterName=DP_WhipNoise_Size})
  ) {materialFunction=/Engine/Functions/Engine_MaterialFunctions02/Texturing/FlowMaps_Simple}
  n13 = Saturate(None=ComponentMask(None=@n12.Diffuse) {g=true})
  n14 = Saturate(None=Multiply(A=Power(Base=@n13)) {constB=4})
  n15 = NamedRerouteUsage()
  n16 = Saturate(None=Power(Base=NamedRerouteUsage()) {constExponent=1.5})
  n17 = Multiply(
   A=NamedRerouteDeclaration(None=Multiply(A=@n14, B=@n15)) {name=DistortedWhisps, nodeColor=<rgba 0 0.4375 1 1>}
   B=Saturate(None=Multiply(A=@n16) {constB=5})
  )
  n18 = NamedRerouteDeclaration(None=@n17) {name=CombinedWhisps, nodeColor=<rgba 0.0312 0 1 1>}
  n19 = NamedRerouteUsage()
  n20 = Saturate(None=Power(Base=@n19) {constExponent=7.5})
  n21 = Add(A=@n20, B=Saturate(None=Power(Base=@n19) {constExponent=15}))
  n22 = LinearInterpolate(
   A=Multiply(
    A=Power(Base=@n3) {constExponent=1.5}
    B=VectorParameter() {group=DarkPit, parameterName=DP_CenterGlow, defaultValue=<rgba 0 0.6557 1 0>}.RGB
   )
   B=VectorParameter() {group=DarkPit, parameterName=DP_WhispColor, defaultValue=<rgba 0.001 0.001 0.001 0>}.RGB
   Alpha=Saturate(None=Subtract(A=@n18, B=@n21))
  )
  n23 = ScalarParameter() {defaultValue=2, group=DarkPit, parameterName=DP_InnerGlowIntensity}
  n24 = SetMaterialAttributes!(
   MaterialAttributes=MaterialFunctionCall(
    PreskinnedPosition=@n2.Pre skinned Local Position
    Gmap=@n2.Pre skinned Local Position
    MaterialAttributes=@n2.Pre skinned Local Position
    AO=@n2.Pre skinned Local Position
   ) {materialFunction=&CharacterSkinFX}.Result
   Base Color=NamedRerouteUsage()
   Specular=LinearInterpolate(Alpha=NamedRerouteUsage()) {constB=0.5}
   Roughness=LinearInterpolate(
    A=ScalarParameter() {defaultValue=0.2, group=DarkPit, parameterName=DP_GlassRoughness}
    B=ScalarParameter() {defaultValue=0.8, group=DarkPit, parameterName=DP_WhispRoughness}
    Alpha=NamedRerouteUsage()
   ) {constA=0.2, constB=0.8}
   Emissive Color=MaterialFunctionCall(
    LDR Input=NamedRerouteDeclaration(None=@n22) {name=BaseColor, nodeColor=<rgba 1 0 0.6562 1>}
    HDR Intensity Multiplier=Multiply(A=@n21, B=@n23) {constB=2}
   ) {materialFunction=&DeriveHDRfromLDR}.RemappedRange
   Opacity Mask=OneMinus(None=TwoSidedSign())
   Normal=LinearInterpolate(A=Constant3Vector() {constant=<rgba 0 0 1 0>}, B=NamedRerouteUsage(), Alpha=NamedRerouteUsage())
  ) {attributeSetTypes [
    #1
    #2
    #3
    #4
    #5
    #6
   ]}
  n25 = MaterialFunctionCall(AO=@n2.AO, Distance Blend=@n2.AO, Fresnel=@n2.AO, Gmap=@n2.AO, Pre Skinned Local Position=@n2.AO, Material=@n24) {materialFunction=&BaseFN_CharacterFunction_PostFX}
  MaterialFunctionCall(In=@n25.Result) {materialFunction=&CINE_MF_CharPostFX}
  n26 = TextureSample() {
   samplerSource=SSM_Wrap_WorldGroupSettings
   samplerType=SAMPLERTYPE_LinearColor
   texture=&T_DarkPit_SwirlFlow
  }
  NamedRerouteDeclaration(None=@n26.B) {name=UV_EdgeMask, nodeColor=<rgba 0 0.2031 1 1>}
  NamedRerouteDeclaration(None=MaterialFunctionCall(Normal=@n12.Normal, Flatness=ScalarParameter() {defaultValue=0.5, group=DarkPit, parameterName=DP_Normal_Flatness}) {materialFunction=&FlattenNormal}.Result) {name=DistortedNormal, nodeColor=<rgba 0.3359 1 0 1>}
  n27 = CameraVectorWS()
  n28 = PixelNormalWS()
  NamedRerouteDeclaration(None=Abs(None=DotProduct(A=@n27, B=@n28))) {name=EdgeMask, nodeColor=<rgba 1 0 0.3281 1>}
  NamedRerouteDeclaration(None=@n20) {name=SpecularFalloffMask, nodeColor=<rgba 0.1953 1 0 1>}
  n29 = Multiply(A=Saturate(None=Subtract(A=@n1) {constB=0.9})) {constB=10}
  Saturate(None=Multiply(A=@n17, B=OneMinus(None=@n29)))
 ]
}
